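# Lit material for floating ice floe pieces that spin slowly and spread apart through world position offset.
# UV1 encodes each piece's pivot, setting its Z-axis rotation center, outward offset and sine-varied spin rate over time.
# Crack and floe normal maps blend with vertex normals in world space, then rotate to match the spin.
# Color blends from time-of-day-tinted water to a top color by world height, darkened by a crack mask.
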Material /Figment_S06/Effects/Fort_Effects/Materials/Ice/M_Ice_Floe_WPO_6_3_1
graph {
 properties {bTangentSpaceNormal=false, pixelDepthOffsetMode=PDOM_Legacy, usedShadingModels=Default Lit}
 nodes [
  n1 = Constant3Vector() {constant=<rgba 0 0 1 0>}
  n2 = ComponentMask(None=TextureCoordinate() {coordinateIndex=1}) {g=true}
  n3 = MaterialFunctionCall(Float2=ObjectPositionWS()) {materialFunction=&BreakOutFloat2Components}
  n4 = Frac(None=Divide(A=@n3.R, B=@n3.R))
  n5 = Multiply(A=Sine(None=Add(A=@n2, B=@n4)) {period=0.2}, B=Multiply(A=Time(), B=ScalarParameter() {defaultValue=0.015, parameterName=SpinSpeed}))
  n6 = ObjectPositionWS()
  n7 = TextureCoordinate() {coordinateIndex=1}
  n8 = AppendVector(A=Multiply(A=Subtract(A=@n7) {constB=0.5}) {constB=2}, B=Constant())
  n9 = Transform(None=Multiply(A=ScalarParameter() {defaultValue=238.2968, parameterName=MeshSize}, B=@n8) {constA=400}) {transformSourceType=TRANSFORMSOURCE_Local}
  Add(
   A=RotateAboutAxis(NormalizedRotationAxis=@n1, RotationAngle=@n5, PivotPoint=Add(A=@n6, B=@n9), Position=WorldPosition().XYZ)
   B=Multiply(A=Transform(None=@n8) {transformSourceType=TRANSFORMSOURCE_Local}, B=ScalarParameter() {defaultValue=50, parameterName=Spread})
  )
  Constant() {r=0.3}
  n10 = TextureCoordinate() {uTiling=2, vTiling=2}
  n11 = Multiply(
   A=TextureSample(UVs=@n10) {samplerType=SAMPLERTYPE_Normal, texture=/Game/Effects/Fort_Effects/Textures/Tile/T_Glass_Cracks_N}.RGB
   B=Constant3Vector() {constant=<rgba 0.5 0.5 0 0>}
  )
  n12 = Multiply(
   A=TextureSample() {samplerType=SAMPLERTYPE_Normal, texture=&/T_Ice_Floe_N}.RGB
   B=VectorParameter() {parameterName=NormalMult, defaultValue=<rgba 1.5 1.5 1 0>}.RGB
  )
  MaterialFunctionCall(
   Rotation Axis=@n1
   Rotation Angle=@n5
   World Space Vertex Normals=MaterialFunctionCall(BaseNormal=VertexNormalWS(), AdditionalNormal=Add(A=@n11, B=@n12)) {materialFunction=&BlendAngleCorrectedNormals}.Result
  ) {materialFunction=&FixRotateAboutAxisNormals}
  Multiply() {constB=0.1}
  n13 = MaterialFunctionCall(
   Color_Morning=VectorParameter() {parameterName=Morning, defaultValue=<rgba 0.1235 0.1422 0.1927 1>}.RGB
   Color_Day=VectorParameter() {parameterName=Day, defaultValue=<rgba 0.0384 0.1056 0.1615 1>}.RGB
   Color_Evening=VectorParameter() {parameterName=Evening, defaultValue=<rgba 0.0648 0.1002 0.131 1>}.RGB
   Color_Night=VectorParameter() {parameterName=Night, defaultValue=<rgba 0.0304 0.0816 0.125 1>}.RGB
  ) {materialFunction=&Color_Time_of_Day}
  n14 = VectorParameter() {parameterName=WaterColor, defaultValue=<rgba 0.2344 0.6214 1 1>}
  n15 = ComponentMask(None=WorldPosition().XYZ) {b=true}
  n16 = ScalarParameter() {defaultValue=5100, parameterName=WaterHeightOffset}
  n17 = Power(
   Base=Divide(A=Add(A=@n15, B=@n16), B=ScalarParameter() {defaultValue=500, parameterName=WaterHeightBlend})
   Exp=ScalarParameter() {defaultValue=1, parameterName=WaterBlendPower}
  )
  Multiply(
   A=LinearInterpolate(
    A=Multiply(A=@n13.Result, B=@n14.RGB)
    B=VectorParameter() {parameterName=TopColor, defaultValue=<rgba 0.3294 0.4111 0.5 0>}.RGB
    Alpha=Clamp(None=@n17)
   )
   B=LinearInterpolate(Alpha=TextureSample(UVs=@n10) {texture=/Game/Effects/Fort_Effects/Textures/Tile/T_Glass_Cracks_M}.G) {constA=1, constB=0.5}
  )
 ]
}
